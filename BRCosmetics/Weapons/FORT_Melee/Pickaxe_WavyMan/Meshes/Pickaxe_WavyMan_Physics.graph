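# Physics asset for the WavyMan two-handed pickaxe, defining bodies and joints for the jiggling character figure attached to its handle.
# A tightly limited spine joins the handle capsule and carries the head and two arms, each arm mostly restricted to a single swing axis.
# The head carries two sphere eyes and three two-segment box hair chains, all with locked twist, giving the eyes and hair the widest swing.
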PhysicsAsset /BRCosmetics/Weapons/FORT_Melee/Pickaxe_WavyMan/Meshes/Pickaxe_WavyMan_Physics
bodies {
 Melee_TwoHanded_Handle [{shapeType=Capsule, length=77.1798, radius=8.0284, center=<0.9243 -1.7587 27.985>, rotation=<rot 0.1741 -90.0007 0.29>}]
 dyn_hairA_01 [{shapeType=Box, extentX=9.5, extentY=1.5, extentZ=2.5}]
 dyn_hairA_02 [{shapeType=Box, extentX=10, extentY=1.5, extentZ=2.5, center=<3.4748 0 0>}]
 dyn_hairB_01 [{shapeType=Box, extentX=9.5, extentY=1.5, extentZ=2.5}]
 dyn_hairB_02 [{shapeType=Box, extentX=7, extentY=1.5, extentZ=2.5, center=<1.9608 0 0>}]
 dyn_hairC_01 [{shapeType=Box, extentX=10, extentY=1.5, extentZ=2.5, center=<0.9459 0 0>}]
 dyn_hairC_02 [{shapeType=Box, extentX=7.5, extentY=1.5, extentZ=2.5, center=<2.0461 0.0002 0>}]
 dyn_head [{shapeType=Capsule, length=6.111, radius=6.6508, center=<0.4105 0.0908 0.5844>, rotation=<rot -4.9958 -9.6991 -0.0136>}]
 dyn_left_arm_01 [{shapeType=Capsule, length=5.8152, radius=2.1965, center=<1.6936 -0.0066 0.0103>, rotation=<rot -9.9278 -89.9999 94.1122>}]
 dyn_left_eye [{shapeType=Sphere, radius=2.75}]
 dyn_right_arm_01 [{shapeType=Capsule, length=5.6748, radius=2.3553, center=<2.4591 -0.0638 0.0886>, rotation=<rot 12.2633 -90 -88.3389>}]
 dyn_right_eye [{shapeType=Sphere, radius=2.75}]
 dyn_spine [{shapeType=Capsule, length=6.3897, radius=6.1917, center=<-0.3108 -0.0159 3.603>, rotation=<rot 1.9767 -94.3331 -0.0004>}]
}
constraints [
 {bone1Name=dyn_head, bone2Name=dyn_spine, swing1LimitDegrees=0, swing2LimitDegrees=7.5, twistLimitDegrees=7.5}
 {bone1Name=dyn_left_eye, bone2Name=dyn_head, swing1LimitDegrees=30, swing2LimitDegrees=30, twistLimitDegrees=45, twistMotion=Locked}
 {bone1Name=dyn_right_eye, bone2Name=dyn_head, swing1LimitDegrees=30, swing2LimitDegrees=30, twistLimitDegrees=45, twistMotion=Locked}
 {bone1Name=dyn_hairA_01, bone2Name=dyn_head, swing1LimitDegrees=45, swing2LimitDegrees=10, twistLimitDegrees=45, twistMotion=Locked}
 {bone1Name=dyn_hairA_02, bone2Name=dyn_hairA_01, swing1LimitDegrees=45, swing2LimitDegrees=10, twistLimitDegrees=45, twistMotion=Locked}
 {bone1Name=dyn_hairB_01, bone2Name=dyn_head, swing1LimitDegrees=45, swing2LimitDegrees=10, twistLimitDegrees=45, twistMotion=Locked}
 {bone1Name=dyn_hairB_02, bone2Name=dyn_hairB_01, swing1LimitDegrees=45, swing2LimitDegrees=10, twistLimitDegrees=45, twistMotion=Locked}
 {bone1Name=dyn_hairC_01, bone2Name=dyn_head, swing1LimitDegrees=45, swing2LimitDegrees=10, twistLimitDegrees=45, twistMotion=Locked}
 {bone1Name=dyn_hairC_02, bone2Name=dyn_hairC_01, swing1LimitDegrees=45, swing2LimitDegrees=10, twistLimitDegrees=45, twistMotion=Locked}
 {bone1Name=dyn_right_arm_01, bone2Name=dyn_spine, swing1LimitDegrees=0, swing2LimitDegrees=25, twistLimitDegrees=0}
 {bone1Name=dyn_spine, bone2Name=Melee_TwoHanded_Handle, swing1LimitDegrees=0, swing2LimitDegrees=2.5, twistLimitDegrees=2.5}
 {bone1Name=dyn_left_arm_01, bone2Name=dyn_spine, swing1LimitDegrees=0, swing2LimitDegrees=25, twistLimitDegrees=0}
]
